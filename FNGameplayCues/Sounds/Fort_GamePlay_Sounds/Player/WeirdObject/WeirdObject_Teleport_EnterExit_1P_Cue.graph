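# Sound cue for the first-person teleport enter and exit effect of the player's WeirdObject.
# A random node picks one of three equally weighted WeirdObject Enter 1P wave variations for each play.
# The cue uses the Fort_Quests sound class with a boosted volume and a very large max distance.
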SoundCue /FNGameplayCues/Sounds/Fort_GamePlay_Sounds/Player/WeirdObject/WeirdObject_Teleport_EnterExit_1P_Cue
root=Random_0
nodes [
 {id=Random_0, children [WavePlayer_1,WavePlayer_2,WavePlayer_3], properties {weights [1,1,1]}}
 {id=WavePlayer_1, properties {soundWaveAssetPtr=&/WeirdObject_Enter_1P_01}}
 {id=WavePlayer_2, properties {soundWaveAssetPtr=&/WeirdObject_Enter_1P_02}}
 {id=WavePlayer_3, properties {soundWaveAssetPtr=&/WeirdObject_Enter_1P_03}}
]
properties {maxDistance=2097152, soundClassObject=&Fort_Quests, volumeMultiplier=1.344}
